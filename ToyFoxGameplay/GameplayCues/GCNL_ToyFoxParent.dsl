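; Parent looping GameplayCue notify actor for ToyFox cues, providing a shared helper to find local players viewing a pilot.
; GetLocalPilotControllers gathers the local Fortnite player controllers and returns the valid ones whose view target carries a primary or secondary pilot tag, or the GiantPawn tag.
; Matching uses the view target rather than the possessed pawn.
; BeginPlay, overlap, Tick and the construction script are empty.
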(event EventBeginPlay)

(event Collision|EventActorBeginOverlap
  (params (OtherActor :type Actor%20Object%20Reference)))

(event EventTick
  (params (DeltaSeconds :type Float%20%28single-precision%29)))

(fn GetLocalPilotControllers
  (params (WorldContextObject :type Object%20Reference))
  (Utilities|FlowControl|Sequence
    (:then_0
      (bind _returnvalue (Fortnite|GetLocalFortPlayerControllers WorldContextObject true true))
      (for _array_element _returnvalue
        (Utilities|IsValid _array_element
          (:Is%20Valid
            (bind _returnvalue_2 (GameplayTags|HasAnyMatchingGameplayTags(Message) (Pawn|GetViewTarget _array_element) (GameplayTags|MakeLiteralGameplayTagContainer "(GameplayTags=((TagName=\"ToyFox.Player.Pilot.Primary\"),(TagName=\"ToyFox.Player.Pilot.Secondary\"),(TagName=\"GiantPawn\")))")))
            (if _returnvalue_2
              (Utilities|Array|Add (LocalVariables|GetReturnControllers) _array_element))))))
    (:then_1
      (return (LocalVariables|GetReturnControllers)))))

(fn ConstructionScript)
