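; Looping gameplay cue for the ToyFox giant pawn's cannon charge, driving its particles, sound and mesh material animation.
; Application spawns charge audio on the head; loop start passes the hotfixable charge time to particles and materials and attaches effects to both muzzles.
; Material stage changes carry game time for shader-driven animation.
; Removal fades valid audio, sets stage two, and culls the first particle component on lower quality levels, including High on console.
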(event EventOnLoopingStart(EitherVFX-preferthisversion)
  (params (MyTarget :type Actor%20Object%20Reference) (Parameters :type Gameplay%20Cue%20Parameters%20Structure%20%28by%20ref%29) (ParticleComponents :type Array%20of%20FXSystem%20Component%20Object%20References) (AudioComponents :type Array%20of%20Audio%20Component%20Object%20References))
  (|CallParentFunction MyTarget Parameters ParticleComponents AudioComponents)
  (Effects|Components|ParticleSystem|SetFloatParameter ParticleComponents "ChargeTime" (ScalableFloat|GetCachedScalableFloatValue (Variables|Hotfix|GetHotfix_CannonChargeTime)))
  (bind _astoy_fox_giant_pawn_interface (Utilities|Casting|CastToToyFoxGiantPawnInterface MyTarget))
  (bind _mesh (Class|ToyFoxGiantPawnInterface|GetSkeletalMeshBudgeted _astoy_fox_giant_pawn_interface))
  (Utilities|IsValid _mesh
    (:Is%20Valid
      (Variables|Default|SetMeshComponent _mesh)
      (Utilities|FlowControl|Sequence
        (:then_0
          (Rendering|Material|SetScalarParameterValueonMaterials (Variables|Default|GetMeshComponent) "Cannon_CurrentStage" 1.0) ; SetStage
          (Rendering|Material|SetScalarParameterValueonMaterials (Variables|Default|GetMeshComponent) "Cannon_AnimDuration" (ScalableFloat|GetCachedScalableFloatValue (Variables|Hotfix|GetHotfix_CannonChargeTime))) ; Set Duration
          (Rendering|Material|SetScalarParameterValueonMaterials (Variables|Default|GetMeshComponent) "Cannon_GameTimeFromBP" (Utilities|Time|GetGameTimeinSeconds (Variables|Default|GetMeshComponent))) ; PassthroughTime
        )
        (:then_1
          (Transformation|AttachComponentToComponent (Utilities|Array|Get(acopy) ParticleComponents) (Variables|Default|GetMeshComponent) "CannonMuzzle_L" "SnapToTarget" "KeepRelative" "KeepRelative" true)
          (Transformation|AttachComponentToComponent (Utilities|Array|Get(acopy) ParticleComponents 1) (Variables|Default|GetMeshComponent) "CannonMuzzle_R" "SnapToTarget" "KeepRelative" "KeepRelative" true))))))

(event EventOnApplication(EitherVFX-preferthisone)
  (params (MyTarget :type Actor%20Object%20Reference) (Parameters :type Gameplay%20Cue%20Parameters%20Structure%20%28by%20ref%29) (ParticleComponents :type Array%20of%20FXSystem%20Component%20Object%20References) (AudioComponents :type Array%20of%20Audio%20Component%20Object%20References) (BurstCameraShakeInstance :type Camera%20Shake%20Base%20Object%20Reference) (BurstDecalInstance :type Decal%20Actor%20Object%20Reference))
  (|CallParentFunction MyTarget Parameters ParticleComponents AudioComponents BurstCameraShakeInstance BurstDecalInstance)
  (bind _ascharacter (Utilities|Casting|CastToCharacter MyTarget))
  (bind _returnvalue (Audio|SpawnSoundAttached :Sound (Variables|Default|GetChargeSoundCue) :AttachToComponent (Variables|/Script/Engine.Character|GetMesh _ascharacter) :AttachPointName "head" :LocationType "SnapToTarget" :bStopWhenAttachedToDestroyed true :VolumeMultiplier 1.0 :PitchMultiplier 1.0 :StartTime 0.0 :bAutoDestroy true))
  (Variables|Default|SetSpawnedChargeAudioComponent _returnvalue))

(event EventOnRemoval(EitherVFX-preferthisversion)
  (params (MyTarget :type Actor%20Object%20Reference) (Parameters :type Gameplay%20Cue%20Parameters%20Structure%20%28by%20ref%29) (ParticleComponents :type Array%20of%20FXSystem%20Component%20Object%20References) (AudioComponents :type Array%20of%20Audio%20Component%20Object%20References) (BurstCameraShakeInstance :type Camera%20Shake%20Base%20Object%20Reference) (BurstDecalInstance :type Decal%20Actor%20Object%20Reference))
  (bind _output (Utilities|Array|Get(acopy) ParticleComponents))
  (bind _meshcomponent (Variables|Default|GetMeshComponent))
  (|CallParentFunction MyTarget Parameters ParticleComponents AudioComponents BurstCameraShakeInstance BurstDecalInstance)
  (bind _spawnedchargeaudiocomponent (Variables|Default|GetSpawnedChargeAudioComponent)
    (:then
      (Audio|Components|Audio|FadeOut _spawnedchargeaudiocomponent 0.2 0.0 "Linear" :!class /Script/Engine.AudioComponent)
      (Utilities|FlowControl|Sequence
        (:then_0
          (Rendering|Material|SetScalarParameterValueonMaterials (Variables|Default|GetMeshComponent) "Cannon_CurrentStage" 2.0) ; SetStage
          (Rendering|Material|SetScalarParameterValueonMaterials (Variables|Default|GetMeshComponent) "Cannon_AnimDuration" (ScalableFloat|GetCachedScalableFloatValue (Variables|Hotfix|GetHotfix_CannonChargeTime))) ; Set Duration
          (Rendering|Material|SetScalarParameterValueonMaterials (Variables|Default|GetMeshComponent) "Cannon_GameTimeFromBP" (Utilities|Time|GetGameTimeinSeconds (Variables|Default|GetMeshComponent))) ; PassthroughTime
        )
        (:then_1
          (Utilities|EffectsQualityCheck 0
            (:Cinematic
              _)
            (:Epic
              _)
            (:High
              (if (Fortnite|IsConsolePlatform)
                (Components|Activation|Deactivate _output :!label _label)
                (Components|DestroyComponent _output 0)
                (else
                  _)))
            (:Medium
              (goto _label))
            (:Low
              (goto _label)))
          (Transformation|AttachComponentToComponent _output _meshcomponent "CannonMuzzle_L" "SnapToTarget" "KeepRelative" "KeepRelative" true)
          (Transformation|AttachComponentToComponent (Utilities|Array|Get(acopy) ParticleComponents 1) _meshcomponent "CannonMuzzle_R" "SnapToTarget" "KeepRelative" "KeepRelative" true))))
    (:else)))

(fn ConstructionScript
  (|CallParentFunction))
