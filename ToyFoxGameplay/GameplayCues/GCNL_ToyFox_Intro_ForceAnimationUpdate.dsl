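; Looping GameplayCue notify for the ToyFox intro that keeps the target player pawn's animation force-ticking while the cue is active.
; On application it caches the pawn, requests a forced anim tick immediately and starts a repeating timer to renew it.
; A hotfixable scalable float sets both timer interval and request duration, since requests are noted as capped at 7 seconds.
; Removal clears the timer, with shipping logs tracing each stage.
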(event EventOnApplication(EitherVFX-preferthisone)
  (params (MyTarget :type Actor%20Object%20Reference) (Parameters :type Gameplay%20Cue%20Parameters%20Structure%20%28by%20ref%29) (ParticleComponents :type Array%20of%20FXSystem%20Component%20Object%20References) (AudioComponents :type Array%20of%20Audio%20Component%20Object%20References) (BurstCameraShakeInstance :type Camera%20Shake%20Base%20Object%20Reference) (BurstDecalInstance :type Decal%20Actor%20Object%20Reference))
  ;[0 0.15 0.016 0.5] OnApplication: Start a timer to force updating the animation every X seconds.
  (|CallParentFunction MyTarget Parameters ParticleComponents AudioComponents BurstCameraShakeInstance BurstDecalInstance)
  (Fortnite|FortShippingLog self "OnApplication: Invoked.")
  (bind _asfort_player_pawn (Utilities|Casting|CastToFortPlayerPawn MyTarget))
  (Variables|Transient|SetTargetPlayerPawn _asfort_player_pawn)
  (bind _returnvalue (Utilities|Time|SetTimerbyEvent (EventDispatchers|CreateEvent :!event Custom|ForceUpdateAnimation) (ScalableFloat|GetCachedScalableFloatValue (Variables|Hotfix|GetHotfix_ForceUpdateAnimationInterval)) true))
  (Variables|Transinet|SetForceUpdateAnimationTimerHandle _returnvalue)
  (CallFunction|ForceUpdateAnimation))

(event Custom|ForceUpdateAnimation
  (bind _hotfix_forceupdateanimationinterval (Variables|Hotfix|GetHotfix_ForceUpdateAnimationInterval))
  (bind _returnvalue (ScalableFloat|GetCachedScalableFloatValue _hotfix_forceupdateanimationinterval))
  ;[0.15 0 0.1409 0.5] ForceUpdateAnimation: Request force animation tick.
  (bind _targetplayerpawn (Variables|Transient|GetTargetPlayerPawn)
    (:then
      (Class|FortPlayerPawn|RequestForceAnimTickDuration _targetplayerpawn _returnvalue) ; We can only request this for 7 seconds at a time\nHardcoded value.
      (Fortnite|FortShippingLog self (Utilities|String|Append "ForceUpdateAnimation: Invoking for " (Utilities|GetDisplayName _targetplayerpawn) " for duration " (Utilities|String|ToString(Float) _returnvalue))))
    (:else)))

(event EventOnRemoval(EitherVFX-preferthisversion)
  (params (MyTarget :type Actor%20Object%20Reference) (Parameters :type Gameplay%20Cue%20Parameters%20Structure%20%28by%20ref%29) (ParticleComponents :type Array%20of%20FXSystem%20Component%20Object%20References) (AudioComponents :type Array%20of%20Audio%20Component%20Object%20References) (BurstCameraShakeInstance :type Camera%20Shake%20Base%20Object%20Reference) (BurstDecalInstance :type Decal%20Actor%20Object%20Reference))
  ;[0.09375 0.04593 0 0.5] OnRemoval: Clean up.
  (|CallParentFunction MyTarget Parameters ParticleComponents AudioComponents BurstCameraShakeInstance BurstDecalInstance)
  (Utilities|Time|ClearandInvalidateTimerbyHandle MyTarget (Variables|Transinet|GetForceUpdateAnimationTimerHandle))
  (Fortnite|FortShippingLog self "OnRemoval: Invoked"))

(fn ConstructionScript
  (|CallParentFunction))
