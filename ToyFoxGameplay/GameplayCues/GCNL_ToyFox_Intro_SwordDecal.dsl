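; Looping gameplay cue for the ToyFox intro that places a sword decal actor on the landscape beneath the giant pawn's sword.
; On application it fades any existing decal, multi-line traces along the FX_SwordCenter socket's up vector, and spawns the decal at landscape hits using the weapon's yaw.
; A trace with no hits retries on a short timer.
; Removal clears that timer and fades out and destroys the decal.
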(event Custom|AttemptTraceAgain
  ;[0 0.15 0.01736 0.5] Application: Spawn the decal under the sword on the landscape.
  (bind _astoy_fox_giant_pawn_interface (Utilities|Casting|CastToToyFoxGiantPawnInterface (stale EventOnApplication(EitherVFX-preferthisone) MyTarget)))
  (bind _weaponmesh (Class|ToyFoxGiantPawnInterface|GetWeaponMeshComponent _astoy_fox_giant_pawn_interface))
  (Utilities|IsValid _weaponmesh
    (:Is%20Valid
      (bind (_returnvalue_location _returnvalue_rotation) (Math|Transform|BreakTransform (Transformation|GetSocketTransform _weaponmesh "FX_SwordCenter" "RTS_World")))
      (bind (_outhits _returnvalue) (Collision|MultiLineTraceForObjects (stale EventOnApplication(EitherVFX-preferthisone) MyTarget) _returnvalue_location (+ _returnvalue_location (* (Math|Vector|GetUpVector _returnvalue_rotation) 3e+03)) (Utilities|Array|MakeArray "ObjectTypeQuery1") false (Utilities|Array|MakeArray (stale EventOnApplication(EitherVFX-preferthisone) MyTarget))))
      (if _returnvalue
        (bind _array_element (Utilities|Array|ForEachLoopwithBreak _outhits)
          (:LoopBody
            (if (Collision|BreakHitResult _array_element)
              (bind _returnvalue_2 (Utilities|IsLandscape (.HitActor (Collision|BreakHitResult _array_element)))
                (:Out
                  (if _returnvalue_2
                    (bind _returnvalue_3 (Game|SpawnActorfromClass (stale EventOnApplication(EitherVFX-preferthisone) MyTarget) "/ToyFoxGameplay/Gameplay/Intro/GameplayAbilities/B_ToyFox_Intro_SwordDecal.B_ToyFox_Intro_SwordDecal_C" (Math|Transform|MakeTransform (.ImpactPoint (Collision|BreakHitResult _array_element)) (Math|Rotator|MakeRotator 0.0 0.0 (.yaw (Transformation|GetWorldRotation _weaponmesh))) "1.000000,1.000000,1.000000") "AlwaysSpawn" "MultiplyWithRoot" 0 0))
                    (Variables|Transient|SetSwordDecalActor _returnvalue_3)))))))
        (else
          (bind _returnvalue_4 (Utilities|Time|SetTimerbyEvent (delegate Custom|AttemptTraceAgain) 0.2))
          (Variables|Transient|SetReattemptTraceTimerHandle _returnvalue_4))))))

(event EventOnApplication(EitherVFX-preferthisone)
  (params (MyTarget :type Actor%20Object%20Reference) (Parameters :type Gameplay%20Cue%20Parameters%20Structure%20%28by%20ref%29) (ParticleComponents :type Array%20of%20FXSystem%20Component%20Object%20References) (AudioComponents :type Array%20of%20Audio%20Component%20Object%20References) (BurstCameraShakeInstance :type Camera%20Shake%20Base%20Object%20Reference) (BurstDecalInstance :type Decal%20Actor%20Object%20Reference))
  ;[0 0.15 0.01736 0.5] Application: Spawn the decal under the sword on the landscape.
  (|CallParentFunction MyTarget Parameters ParticleComponents AudioComponents BurstCameraShakeInstance BurstDecalInstance)
  (Utilities|FlowControl|Sequence
    (:then_0
      (bind _sworddecalactor (Variables|Transient|GetSwordDecalActor)
        (:then
          (Class|BToyFoxIntroSwordDecal|FadeOutDecalandDestroyActor _sworddecalactor))
        (:else)))
    (:then_1
      (bind _astoy_fox_giant_pawn_interface (Utilities|Casting|CastToToyFoxGiantPawnInterface MyTarget))
      (bind _weaponmesh (Class|ToyFoxGiantPawnInterface|GetWeaponMeshComponent _astoy_fox_giant_pawn_interface))
      (Utilities|IsValid _weaponmesh
        (:Is%20Valid
          (bind (_returnvalue_location _returnvalue_rotation) (Math|Transform|BreakTransform (Transformation|GetSocketTransform _weaponmesh "FX_SwordCenter" "RTS_World")))
          (bind (_outhits _returnvalue) (Collision|MultiLineTraceForObjects MyTarget _returnvalue_location (+ _returnvalue_location (* (Math|Vector|GetUpVector _returnvalue_rotation) 3e+03)) (Utilities|Array|MakeArray "ObjectTypeQuery1") false (Utilities|Array|MakeArray MyTarget)))
          (if _returnvalue
            (bind _array_element (Utilities|Array|ForEachLoopwithBreak _outhits)
              (:LoopBody
                (if (Collision|BreakHitResult _array_element)
                  (bind _returnvalue_2 (Utilities|IsLandscape (.HitActor (Collision|BreakHitResult _array_element)))
                    (:Out
                      (if _returnvalue_2
                        (bind _returnvalue_3 (Game|SpawnActorfromClass MyTarget "/ToyFoxGameplay/Gameplay/Intro/GameplayAbilities/B_ToyFox_Intro_SwordDecal.B_ToyFox_Intro_SwordDecal_C" (Math|Transform|MakeTransform (.ImpactPoint (Collision|BreakHitResult _array_element)) (Math|Rotator|MakeRotator 0.0 0.0 (.yaw (Transformation|GetWorldRotation _weaponmesh))) "1.000000,1.000000,1.000000") "AlwaysSpawn" "MultiplyWithRoot" 0 0))
                        (Variables|Transient|SetSwordDecalActor _returnvalue_3)))))))
            (else
              (bind _returnvalue_4 (Utilities|Time|SetTimerbyEvent (delegate Custom|AttemptTraceAgain) 0.2))
              (Variables|Transient|SetReattemptTraceTimerHandle _returnvalue_4))))))))

(event EventOnRemoval(EitherVFX-preferthisversion)
  (params (MyTarget :type Actor%20Object%20Reference) (Parameters :type Gameplay%20Cue%20Parameters%20Structure%20%28by%20ref%29) (ParticleComponents :type Array%20of%20FXSystem%20Component%20Object%20References) (AudioComponents :type Array%20of%20Audio%20Component%20Object%20References) (BurstCameraShakeInstance :type Camera%20Shake%20Base%20Object%20Reference) (BurstDecalInstance :type Decal%20Actor%20Object%20Reference))
  ;[0.08854 0.05137 0 0.5] Removal: Fade out and destroy the decal.
  (|CallParentFunction MyTarget Parameters ParticleComponents AudioComponents BurstCameraShakeInstance BurstDecalInstance)
  (Utilities|Time|ClearandInvalidateTimerbyHandle MyTarget (Variables|Transient|GetReattemptTraceTimerHandle))
  (bind _sworddecalactor (Variables|Transient|GetSwordDecalActor)
    (:then
      (Class|BToyFoxIntroSwordDecal|FadeOutDecalandDestroyActor _sworddecalactor)
      (Variables|Transient|SetSwordDecalActor 0))
    (:else)))

(fn ConstructionScript
  (|CallParentFunction))
