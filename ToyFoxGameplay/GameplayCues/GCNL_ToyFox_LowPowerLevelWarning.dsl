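; Looping gameplay cue for ToyFox that shows a low power level screen warning to local players viewing the cue's target.
; It binds each local controller's view-target change to add or destroy a camera lens effect and toggle a material parameter collection's WarningStart.
; Effect level 1 marks a secondary pilot, skipping the lens's InitialPostMaterial call.
; Removal unbinds and destroys mapped lens actors; start and removal print debug strings.
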(event EventOnLoopingStart(EitherVFX-preferthisversion)
  (params (MyTarget :type Actor%20Object%20Reference) (Parameters :type Gameplay%20Cue%20Parameters%20Structure%20%28by%20ref%29) (ParticleComponents :type Array%20of%20FXSystem%20Component%20Object%20References) (AudioComponents :type Array%20of%20Audio%20Component%20Object%20References))
  (bind (_normalizedmagnitude _rawmagnitude _effectcontext _matchedtagname _originaltag _aggregatedsourcetags _aggregatedtargettags _location _normal _instigator _effectcauser _sourceobject _physicalmaterial _gameplayeffectlevel) (Ability|GameplayCue|BreakGameplayCueParameters Parameters))
  ;[0 0.15 0.01562 0.5] OnLoopingStart: Get all the local controllers (this is just one in most cases but can be 2 in splitscreen) and bind to know when their view target changes.
  (|CallParentFunction MyTarget Parameters ParticleComponents AudioComponents)
  (bind _output_get (Variables|Transient|SetTargetActor MyTarget))
  (Development|PrintString _output_get "APPLY LOW LEVEL WARNING" true true "(R=1.000000,G=0.000000,B=0.035143,A=1.000000)" 1e+01)
  (Variables|Transient|SetbSecondaryPilot (== _gameplayeffectlevel 1))
  (for _array_element (Fortnite|GetLocalFortPlayerControllers MyTarget true true)
    (Utilities|IsValid _array_element
      (:Is%20Valid
        (EventDispatchers|BindEventtoOnViewTargetChanged _array_element (EventDispatchers|CreateEvent :!event Custom|HandleLocalPlayerControllerViewTargetChanged))
        (CallFunction|HandleLocalPlayerControllerViewTargetChanged _array_element 0 (Pawn|GetViewTarget _array_element)) ; Run logic for current view target.
      ))))

(event EventOnRemoval(EitherVFX-preferthisversion)
  (params (MyTarget :type Actor%20Object%20Reference) (Parameters :type Gameplay%20Cue%20Parameters%20Structure%20%28by%20ref%29) (ParticleComponents :type Array%20of%20FXSystem%20Component%20Object%20References) (AudioComponents :type Array%20of%20Audio%20Component%20Object%20References) (BurstCameraShakeInstance :type Camera%20Shake%20Base%20Object%20Reference) (BurstDecalInstance :type Decal%20Actor%20Object%20Reference))
  (bind _controllertocameralensactormap (Variables|Transient|GetControllerToCameraLensActorMap))
  ;[0.07292 0.04927 0 0.5] OnRemoval: Unbind from all local player controllers and destroy and camera lens effect actors spawned for them.
  (|CallParentFunction MyTarget Parameters ParticleComponents AudioComponents BurstCameraShakeInstance BurstDecalInstance)
  (Development|PrintString MyTarget "Remove LOW LEVEL WARNING" true true "(R=0.229930,G=1.000000,B=0.000000,A=1.000000)" 1e+01)
  (Utilities|FlowControl|Sequence
    (:then_0
      (for _array_element (Fortnite|GetLocalFortPlayerControllers MyTarget true true)
        (Utilities|IsValid _array_element
          (:Is%20Valid
            (EventDispatchers|UnbindEventfromOnViewTargetChanged _array_element (EventDispatchers|CreateEvent :!event Custom|HandleLocalPlayerControllerViewTargetChanged))))))
    (:then_1
      (bind _keys (Utilities|Map|Keys _controllertocameralensactormap))
      (for _array_element_2 _keys
        (Utilities|IsValid (Utilities|Map|Find _controllertocameralensactormap _array_element_2)
          (:Is%20Valid
            (Actor|DestroyActor (Utilities|Map|Find _controllertocameralensactormap _array_element_2))))))))

(event Custom|HandleLocalPlayerControllerViewTargetChanged
  (params (PlayerController :type Fort%20Player%20Controller%20Object%20Reference) (OldViewTarget :type Actor%20Object%20Reference) (NewViewTarget :type Actor%20Object%20Reference))
  (bind _controllertocameralensactormap (Variables|Transient|GetControllerToCameraLensActorMap))
  (bind _value (Utilities|Map|Find _controllertocameralensactormap PlayerController))
  ;[0 0.133 0.15 0.5] HandleLocalPlayerControllerViewTargetChanged: When the view target of the local player changes, if they end up viewing the Target Actor of this Gameplay Cue, then add the screen warning camera lens effect.
  ; IJf they end up viewing an actor that is NOT the Target Actor of this Gameplay Cue, then destroy the camera lens so they don't see it when viewing other players.
  (Utilities|IsValid PlayerController
    (:Is%20Valid
      (bind _playercameramanager (Variables|/Script/Engine.PlayerController|GetPlayerCameraManager PlayerController)
        (:then
          (if (== (Variables|Transient|GetTargetActor) NewViewTarget)
            (bind _returnvalue (Camera|AddGenericCameraLensEffect _playercameramanager "/Game/Effects/Fort_Effects/Camera_Effects/BP_CameraLens_ScreenWarning.BP_CameraLens_ScreenWarning_C")) ; This doesn't add duplicate lens effect unless ShouldAllowMultipleInstances is TRUE, which isn't for the applied camera effect.
            (bind _asactor (Utilities|Casting|CastToActor _returnvalue))
            (Utilities|Map|Add (Variables|Transient|GetControllerToCameraLensActorMap) PlayerController _asactor)
            (Rendering|Material|SetScalarParameterValue _asactor "/ToyFoxGameplay/Assets/FX/Materials/MPC_ScreenHazeUp.MPC_ScreenHazeUp" "WarningStart" 1.0 :!class /Script/Engine.KismetMaterialLibrary)
            ;;[0.1023 0 0.15 0.5] DO LOGIC FOR SECONDARY PILOT HERE AFTER THE LENS HAS BEEN ADDED.
            (if (not (Variables|Transient|GetbSecondaryPilot))
              (bind _asbp_camera_lens_screen_warning (Utilities|Casting|CastToBP_CameraLens_ScreenWarning _asactor))
              (Class|BPCameraLensScreenWarning|InitialPostMaterial _asbp_camera_lens_screen_warning))
            (else
              (Rendering|Material|SetScalarParameterValue PlayerController "/ToyFoxGameplay/Assets/FX/Materials/MPC_ScreenHazeUp.MPC_ScreenHazeUp" "WarningStart" :!class /Script/Engine.KismetMaterialLibrary)
              (Utilities|IsValid _value
                (:Is%20Valid
                  (Actor|DestroyActor _value))))))
        (:else)))))

(fn ConstructionScript
  (|CallParentFunction))
