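; Looping gameplay cue parent for ToyFox melee swings that adds a sound indicator when the cue is applied.
; After the parent's application effects run, it reads the first burst sound and, if valid, waits a configurable delay.
; If the target is still valid, it shows a "Gunshot" indicator at the target to local pilot controllers for friendly, neutral and hostile teams, using the sound's max distance as range.
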(event EventOnApplication(EitherVFX-preferthisone)
  (params (MyTarget :type Actor%20Object%20Reference) (Parameters :type Gameplay%20Cue%20Parameters%20Structure%20%28by%20ref%29) (ParticleComponents :type Array%20of%20FXSystem%20Component%20Object%20References) (AudioComponents :type Array%20of%20Audio%20Component%20Object%20References) (BurstCameraShakeInstance :type Camera%20Shake%20Base%20Object%20Reference) (BurstDecalInstance :type Decal%20Actor%20Object%20Reference))
  (bind _applicationburst (Variables|ApplicationEffects|GetApplicationBurst))
  (bind (_burstparticles _burstsounds) (Utilities|Struct|BreakFortBurstEffectData _applicationburst))
  (bind _output (Utilities|Array|Get(aref) _burstsounds))
  (bind (_condition _attachment _soundcue) (Utilities|Struct|BreakFortGameplayCueAudioInfo _output))
  ;[0.15 0 0.1043 0.5] OnApplication: Show the sound indicator for the sword swing.
  (|CallParentFunction MyTarget Parameters ParticleComponents AudioComponents BurstCameraShakeInstance BurstDecalInstance)
  (if (Utilities|Array|IsValidIndex _burstsounds)
    (Utilities|IsValid _soundcue
      (:Is%20Valid
        (Utilities|FlowControl|Delay MyTarget (Variables|SoundIndicator|GetSoundIndicatorDelay))
        (Utilities|IsValid MyTarget
          (:Is%20Valid
            (Fortnite|ShowSoundIndicatorLocalforTeam MyTarget MyTarget (Transformation|GetActorLocation MyTarget) (Variables|/Script/Engine.SoundBase|GetMaxDistance _soundcue) "Gunshot" (Default|GetLocalPilotControllers MyTarget) (Team|GetActorTeam MyTarget) (Utilities|Array|MakeArray "Friendly" "Neutral" "Hostile") MyTarget "(R=0.000000,G=0.000000,B=0.000000,A=0.000000)" "" _soundcue)))))))

(fn ConstructionScript
  (|CallParentFunction))
