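; Parent gameplay cue for ToyFox, supplying helper functions that target local players for pilot-specific and camera feedback.
; GetLocalPilotControllers collects local player controllers whose view target carries a primary pilot, secondary pilot or GiantPawn gameplay tag.
; PlayLocalCameraShake starts a given shake class in camera-local space, only on local controllers viewing the target Fortnite pawn when it is locally controlled or viewed.
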(fn GetLocalPilotControllers
  (params (WorldContextObject :type Object%20Reference))
  (Utilities|FlowControl|Sequence
    (:then_0
      (bind _returnvalue (Fortnite|GetLocalFortPlayerControllers WorldContextObject true true))
      (for _array_element _returnvalue
        (Utilities|IsValid _array_element
          (:Is%20Valid
            (bind _returnvalue_2 (GameplayTags|HasAnyMatchingGameplayTags(Message) (Pawn|GetViewTarget _array_element) (GameplayTags|MakeLiteralGameplayTagContainer "(GameplayTags=((TagName=\"ToyFox.Player.Pilot.Primary\"),(TagName=\"ToyFox.Player.Pilot.Secondary\"),(TagName=\"GiantPawn\")))")))
            (if _returnvalue_2
              (Utilities|Array|Add (LocalVariables|GetReturnControllers) _array_element))))))
    (:then_1
      (return (LocalVariables|GetReturnControllers)))))

(fn PlayLocalCameraShake
  (params (TargetActor :type Actor%20Object%20Reference) (ShakeClass :type Camera%20Shake%20Base%20Class%20Reference))
  (bind _asfort_player_pawn (Utilities|Casting|CastToFortPlayerPawn TargetActor))
  (if (or (Pawn|IsLocallyControlled _asfort_player_pawn) (Pawn|IsLocallyViewed _asfort_player_pawn))
    (for _array_element (Fortnite|GetLocalFortPlayerControllers _asfort_player_pawn)
      (Utilities|IsValid _array_element
        (:Is%20Valid
          (if (== (Pawn|GetViewTarget _array_element) _asfort_player_pawn)
            (bind _playercameramanager (Variables|/Script/Engine.PlayerController|GetPlayerCameraManager _array_element)
              (:then
                (CameraShakes|StartCameraShake :self _playercameramanager :ShakeClass ShakeClass :Scale 1.0 :PlaySpace "CameraLocal"))
              (:else))))))))
