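; First-person burst gameplay cue for the ToyFox giant pawn's cannon firing.
; After the parent burst runs, it gets the pawn's budgeted skeletal mesh through the giant pawn interface and sets cannon stage, duration and current game time as material parameters, apparently for shader-timed animation.
; It also spawns attached Niagara muzzle flashes at the left and right muzzle sockets and plays a local cannon-fire camera shake.
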(fn OnBurst(EitherVFX-preferthisversion)
  (params (MyTarget :type Actor%20Object%20Reference) (Parameters :type Gameplay%20Cue%20Parameters%20Structure%20%28by%20ref%29) (ParticleComponents :type Array%20of%20FXSystem%20Component%20Object%20References) (AudioComponents :type Array%20of%20Audio%20Component%20Object%20References) (BurstCameraShakeInstance :type Camera%20Shake%20Base%20Object%20Reference) (BurstDecalInstance :type Decal%20Actor%20Object%20Reference))
  (|CallParentFunction MyTarget Parameters ParticleComponents AudioComponents BurstCameraShakeInstance BurstDecalInstance)
  (Utilities|FlowControl|Sequence
    (:then_0
      (bind _astoy_fox_giant_pawn_interface (Utilities|Casting|CastToToyFoxGiantPawnInterface MyTarget))
      (bind _mesh (Class|ToyFoxGiantPawnInterface|GetSkeletalMeshBudgeted _astoy_fox_giant_pawn_interface))
      (Utilities|IsValid _mesh
        (:Is%20Valid
          (Utilities|FlowControl|Sequence
            (:then_0
              (Rendering|Material|SetScalarParameterValueonMaterials _mesh "Cannon_CurrentStage" 2.0) ; SetStage
              (Rendering|Material|SetScalarParameterValueonMaterials _mesh "Cannon_AnimDuration" 2.0) ; Set Duration
              (Rendering|Material|SetScalarParameterValueonMaterials _mesh "Cannon_GameTimeFromBP" (Utilities|Time|GetGameTimeinSeconds _mesh)) ; PassthroughTime
            )
            (:then_1
              (Niagara|SpawnSystemAttached "/ToyFoxGameplay/Assets/FX/Cannon/NS_ToyFox_Cannon_MuzzleFlash.NS_ToyFox_Cannon_MuzzleFlash" _mesh "CannonMuzzle_R" "0, 0, 0" "0, 0, 0" "SnapToTarget" true true "AutoRelease")
              (Niagara|SpawnSystemAttached "/ToyFoxGameplay/Assets/FX/Cannon/NS_ToyFox_Cannon_MuzzleFlash.NS_ToyFox_Cannon_MuzzleFlash" _mesh "CannonMuzzle_L" "0, 0, 0" "0, 0, 0" "SnapToTarget" true true "AutoRelease"))))
        (:Is%20Not%20Valid
          _))
      (CallFunction|PlayLocalCameraShake MyTarget "/Game/Effects/Fort_Effects/Camera_Effects/BP_CameraShake_Cannon_Fire.BP_CameraShake_Cannon_Fire_C"))
    (:then_1
      (return))))
