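; Burst gameplay cue that plays the water impact effect for a ToyFox cannon projectile hit.
; On burst it flags the first Niagara component as water, then enlarges the second, disables its water spline mesh and pitches it 90 degrees.
; If the r.PhysicsField.EnableField console variable is on, it also spawns an FN_RadialForce physics field actor at the first component's location.
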(fn OnBurst(Niagara)
  (params (MyTarget :type Actor%20Object%20Reference) (Parameters :type Gameplay%20Cue%20Parameters%20Structure%20%28by%20ref%29) (ParticleComponents :type Array%20of%20Niagara%20Particle%20System%20Component%20Object%20References) (AudioComponents :type Array%20of%20Audio%20Component%20Object%20References) (BurstCameraShakeInstance :type Camera%20Shake%20Base%20Object%20Reference) (BurstDecalInstance :type Decal%20Actor%20Object%20Reference))
  (bind _output (Utilities|Array|Get(acopy) ParticleComponents 1))
  (bind _output_2 (Utilities|Array|Get(acopy) ParticleComponents))
  (Effects|Components|ParticleSystem|SetBoolParameter _output_2 "IsWater" true)
  (Effects|Components|ParticleSystem|SetFloatParameter _output "EffectScale" 3.0)
  (Effects|Components|ParticleSystem|SetBoolParameter _output "SpawnWaterSplineMesh" false)
  (Transformation|AddRelativeRotation _output "90.000000,0.000000,0.000000" false false)
  (bind _returnvalue (Development|GetConsoleVariableIntValue "r.PhysicsField.EnableField"))
  (if (== _returnvalue 1)
    (Game|SpawnActorfromClass 0 "/Game/Effects/Niagara/Blueprints/PhsyicsFields/FN_RadialForce.FN_RadialForce_C" (Math|Transform|MakeTransform (Transformation|GetWorldLocation _output_2)) "Undefined" "MultiplyWithRoot" 0 0.4 0.75 2.5e+03 1.4 0.0 0)
    (else
      _))
  (return))
