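; Burst gameplay cue that plays the sword power-up visual for the ToyFox giant pawn's finisher.
; On burst it gets the weapon and body meshes through the giant pawn interface, then passes game time, duration and curve flags to their materials, so weapon and eye animations run in-shader.
; It also attaches the sword power-up Niagara effect to the weapon mesh.
; An invalid weapon mesh instead zeroes the weapon curve and eye stage parameters.
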(fn OnBurst(EitherVFX-preferthisversion)
  (params (MyTarget :type Actor%20Object%20Reference) (Parameters :type Gameplay%20Cue%20Parameters%20Structure%20%28by%20ref%29) (ParticleComponents :type Array%20of%20FXSystem%20Component%20Object%20References) (AudioComponents :type Array%20of%20Audio%20Component%20Object%20References) (BurstCameraShakeInstance :type Camera%20Shake%20Base%20Object%20Reference) (BurstDecalInstance :type Decal%20Actor%20Object%20Reference))
  (bind _returnvalue (Utilities|Time|GetGameTimeinSeconds 0))
  (|CallParentFunction MyTarget Parameters ParticleComponents AudioComponents BurstCameraShakeInstance BurstDecalInstance)
  (Utilities|FlowControl|Sequence
    (:then_0
      (bind _astoy_fox_giant_pawn_interface (Utilities|Casting|CastToToyFoxGiantPawnInterface MyTarget))
      (bind _weaponmesh (Class|ToyFoxGiantPawnInterface|GetWeaponMeshComponent _astoy_fox_giant_pawn_interface))
      (bind _mesh (Class|ToyFoxGiantPawnInterface|GetSkeletalMeshBudgeted _astoy_fox_giant_pawn_interface))
      (Utilities|IsValid _weaponmesh
        (:Is%20Valid
          (Utilities|FlowControl|Sequence
            (:then_0
              ;[0 0.15 0.01121 0.5] Process Animation Curve in material
              (Rendering|Material|SetScalarParameterValueonMaterials _weaponmesh "GameTimeFromBP" _returnvalue) ; Passthough Time
              (Rendering|Material|SetScalarParameterValueonMaterials _weaponmesh "Animation_Duration" 3.0) ; Set Duration. 3s
              (Rendering|Material|SetScalarParameterValueonMaterials _weaponmesh "ProcessAnimationCurve" 1.0) ; Process Animation Curve to 1
              (Rendering|Material|SetScalarParameterValueonMaterials _mesh "Eye_GameTimeFromBP" _returnvalue) ; Passthough Time
              (Rendering|Material|SetScalarParameterValueonMaterials _mesh "Eye_AnimDuration" 3.0) ; Set Duration. 3s
              (Rendering|Material|SetScalarParameterValueonMaterials _mesh "Eye_CurrentStage" 1.0) ; Process Animation Curve to 1
            )
            (:then_1
              (Effects|SpawnFXSystemAttached :FXSystemAsset "/ToyFoxGameplay/Assets/FX/NS_ToyFox_SwordPowerUp.NS_ToyFox_SwordPowerUp" :AttachToComponent _weaponmesh :AttachPointName "weapon_r" :Scale "1.000000,1.000000,1.000000" :LocationType "SnapToTarget" :bAutoDestroy true :PoolingMethod "None" :bAutoActivate true))))
        (:Is%20Not%20Valid
          (Rendering|Material|SetScalarParameterValueonMaterials _weaponmesh "ProcessAnimationCurve" 0.0) ; Reset Animation Curve and stop at 0
          (Rendering|Material|SetScalarParameterValueonMaterials _mesh "Eye_CurrentStage" 0.0) ; Process Animation Curve to 1
        )))
    (:then_1
      (return))))
