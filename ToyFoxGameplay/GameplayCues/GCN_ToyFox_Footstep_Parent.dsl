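; Base burst gameplay cue for ToyFox footsteps, playing stomp effects at a configurable foot socket on the player pawn.
; After the parent burst, a hit on a water body spawns a borrowed biplane splash, oriented by foot side, and destroys the burst particle.
; It also plays the footstep sound and world camera shake at the socket.
; A sound indicator for hostile and neutral teams uses the sound's max distance as range.
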(fn OnBurst(EitherVFX-preferthisversion)
  (params (MyTarget :type Actor%20Object%20Reference) (Parameters :type Gameplay%20Cue%20Parameters%20Structure%20%28by%20ref%29) (ParticleComponents :type Array%20of%20FXSystem%20Component%20Object%20References) (AudioComponents :type Array%20of%20Audio%20Component%20Object%20References) (BurstCameraShakeInstance :type Camera%20Shake%20Base%20Object%20Reference) (BurstDecalInstance :type Decal%20Actor%20Object%20Reference))
  (bind (_normalizedmagnitude _rawmagnitude _effectcontext) (Ability|GameplayCue|BreakGameplayCueParameters Parameters))
  (bind _returnvalue (Ability|EffectContext|GetHitResult _effectcontext))
  (bind (_bblockinghit _binitialoverlap _time _distance _location _impactpoint _normal _impactnormal _physmat _hitactor) (Collision|BreakHitResult _returnvalue))
  (bind _output (Utilities|Array|Get(acopy) ParticleComponents))
  (bind _footsocketname (Variables|Default|GetFootSocketName))
  (|CallParentFunction MyTarget Parameters ParticleComponents AudioComponents BurstCameraShakeInstance BurstDecalInstance)
  (Utilities|FlowControl|Sequence
    (:then_0
      (Utilities|FlowControl|Sequence
        (:then_0
          (bind _asfort_player_pawn (Utilities|Casting|CastToFortPlayerPawn MyTarget))
          (bind _mesh (Variables|/Script/Engine.Character|GetMesh _asfort_player_pawn)
            (:then
              (LocalVariables|SetStompFootSocketRotation (Transformation|GetSocketRotation _mesh (Variables|Default|GetFootSocketName))))
            (:else)))
        (:then_1
          (bind _asfort_water_body_actor (Utilities|Casting|CastToFortWaterBodyActor _hitactor)) ; Footstep impact was on water.
          (bind _outwatersurfaceinfo (WaterBody|GetWaterSurfaceInfoStruct _asfort_water_body_actor _impactpoint true false false))
          (bind _returnvalue_2 (Effects|Components|ParticleSystem|SpawnFXAtLocation 0 "/Game/Athena/DrivableVehicles/FX/Plane/WaterInteraction/NS_Biplane_InWater_HugeSplash_Opt.NS_Biplane_InWater_HugeSplash_Opt" (.WaterSurfaceLocation (Utilities|Struct|BreakWaterSurfaceInfo _outwatersurfaceinfo)) (Math|Rotator|MakeRotfromX (* (Math|Vector|GetRightVector (LocalVariables|GetStompFootSocketRotation)) (select (== (Variables|Default|GetFootSocketName) "Foot_L") 1 -1)))))
          (Effects|Components|ParticleSystem|SetFloatParameter _returnvalue_2 "EffectScale" 1.5)
          (Components|Activation|Deactivate _output)
          (Components|DestroyComponent _output 0))
        (:then_2
          (bind _asfort_player_pawn_2 (Utilities|Casting|CastToFortPlayerPawn MyTarget))
          (bind _mesh_2 (Variables|/Script/Engine.Character|GetMesh _asfort_player_pawn_2)
            (:then
              (Audio|SpawnSoundatLocation _asfort_player_pawn_2 (Variables|Default|GetFootstepSoundCue) (Transformation|GetSocketLocation _mesh_2 _footsocketname))
              (Camera|PlayWorldCameraShake _asfort_player_pawn_2 (Variables|Default|GetGroundImpactCameraShake) (Transformation|GetSocketLocation _mesh_2 _footsocketname) (Variables|Default|GetGroundImpactCameraShakeRadiusInner) (Variables|Default|GetGroundImpactCameraShakeRadiusOuter))
              (Fortnite|ShowSoundIndicatorLocalforTeam _asfort_player_pawn_2 _asfort_player_pawn_2 (Transformation|GetSocketLocation _mesh_2 _footsocketname) (Variables|/Script/Engine.SoundBase|GetMaxDistance (Variables|Default|GetFootstepSoundCue)) "Gunshot" (CallFunction|GetLocalPilotControllers _asfort_player_pawn_2) (Team|GetActorTeam _asfort_player_pawn_2) (Utilities|Array|MakeArray "Hostile" "Neutral") _asfort_player_pawn_2 "(R=0.000000,G=0.000000,B=0.000000,A=0.000000)" "" (Variables|Default|GetFootstepSoundCue)))
            (:else)))))
    (:then_1
      (return))))
