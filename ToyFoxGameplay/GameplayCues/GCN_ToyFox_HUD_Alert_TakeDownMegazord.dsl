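; Burst gameplay cue for ToyFox that raises the Rita "Take Down Megazord" HUD alert for the local viewer.
; On burst, it casts the target to a Fort player pawn and continues only if that pawn is locally viewed.
; It finds the ToyFox game state component on the Athena game state and broadcasts an alert notification struct carrying the alert tag and that pawn's player state.
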(fn OnBurst(EitherVFX-preferthisversion)
  (params (MyTarget :type Actor%20Object%20Reference) (Parameters :type Gameplay%20Cue%20Parameters%20Structure%20%28by%20ref%29) (ParticleComponents :type Array%20of%20FXSystem%20Component%20Object%20References) (AudioComponents :type Array%20of%20Audio%20Component%20Object%20References) (BurstCameraShakeInstance :type Camera%20Shake%20Base%20Object%20Reference) (BurstDecalInstance :type Decal%20Actor%20Object%20Reference))
  (Utilities|FlowControl|Sequence
    (:then_0
      (bind _asfort_player_pawn (Utilities|Casting|CastToFortPlayerPawn MyTarget))
      (if (Pawn|IsLocallyViewed _asfort_player_pawn)
        (Utilities|IsValid (Fortnite|Athena|GameState|GetGameStateAthena _asfort_player_pawn)
          (:Is%20Valid
            (Utilities|IsValid (Actor|GetComponentbyClass (Fortnite|Athena|GameState|GetGameStateAthena _asfort_player_pawn) "/ToyFoxGameplay/Gameplay/B_ToyFoxGameStateComponent.B_ToyFoxGameStateComponent_C")
              (:Is%20Valid
                (bind _returnvalue (Fortnite|Utilities|AttemptGetPlayerStateforPawn _asfort_player_pawn))
                (bind _asfort_player_state_athena (Utilities|Casting|CastToFortPlayerStateAthena _returnvalue))
                (Class|BToyFoxGameStateComponent|BroadcastAlertNotification (Actor|GetComponentbyClass (Fortnite|Athena|GameState|GetGameStateAthena _asfort_player_pawn) "/ToyFoxGameplay/Gameplay/B_ToyFoxGameStateComponent.B_ToyFoxGameStateComponent_C") (Utilities|Struct|MakeAlertNotificationStruct "(TagName=\"ToyFox.HUD.Alert.Rita.TakeDownMegazord\")" (Utilities|Array|MakeArray _asfort_player_state_athena)))))))))
    (:then_1
      (return))))
