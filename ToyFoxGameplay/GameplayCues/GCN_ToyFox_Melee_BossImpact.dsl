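; ToyFox gameplay cue notify for the burst effects of a melee boss impact.
; OnBurst runs the parent logic, then line-traces against the hit component along the cue's hit direction to move particles onto its body surface.
; It then orients the first particle component along the target pawn's aim right vector, using the cue's truncated raw magnitude to choose left or right.
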(fn OnBurst(EitherVFX-preferthisversion)
  (params (MyTarget :type Actor%20Object%20Reference) (Parameters :type Gameplay%20Cue%20Parameters%20Structure%20%28by%20ref%29) (ParticleComponents :type Array%20of%20FXSystem%20Component%20Object%20References) (AudioComponents :type Array%20of%20Audio%20Component%20Object%20References) (BurstCameraShakeInstance :type Camera%20Shake%20Base%20Object%20Reference) (BurstDecalInstance :type Decal%20Actor%20Object%20Reference))
  (bind (_normalizedmagnitude _rawmagnitude _effectcontext) (Ability|GameplayCue|BreakGameplayCueParameters Parameters))
  (bind _returnvalue (Ability|EffectContext|GetHitResult _effectcontext))
  (bind (_bblockinghit _binitialoverlap _time _distance _location _impactpoint _normal _impactnormal _physmat _hitactor _hitcomponent _hitbonename _bonename _hititem _elementindex _faceindex _tracestart) (Collision|BreakHitResult _returnvalue))
  ;[0.03178 0.07964 0.15 0.5] OnBurst: Running a line trace to help play the FX on the body of the hit component instead of within the actor.
  (|CallParentFunction MyTarget Parameters ParticleComponents AudioComponents BurstCameraShakeInstance BurstDecalInstance)
  (Utilities|FlowControl|Sequence
    (:then_0
      (Utilities|IsValid _hitcomponent
        (:Is%20Valid
          (bind (_hitlocation _hitnormal _bonename_2 _outhit _returnvalue_2) (Collision|LineTraceComponent _hitcomponent _tracestart (+ (* (Math|Vector|GetUnitDirection(Vector) _tracestart _impactpoint) (+ _distance 1e+03)) _tracestart) true false false))
          (if _returnvalue_2
            (Transformation|SetWorldLocation ParticleComponents _hitlocation false false)))))
    (:then_1
      (bind _aspawn (Utilities|Casting|CastToPawn MyTarget))
      (Transformation|SetWorldRotation (Utilities|Array|Get(acopy) ParticleComponents) (Math|Conversions|RotationFromXVector (* (Math|Vector|GetRightVector (Pawn|GetBaseAimRotation _aspawn)) (Math|Float|SelectFloat -1.0 1.0 (== (Math|Float|Truncate _rawmagnitude) 1)))) false false))
    (:then_2
      (return))))
