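; Burst gameplay cue for ToyFox's melee building destruction that orients the spawned effect sideways to the attacker's aim.
; After the parent burst runs, it rotates the first particle component along the target pawn's aim right vector.
; The truncated raw magnitude acts as a direction flag, with a value of 1 flipping that vector.
; It also copies the effect context origin into a HitBuildings_BoxExtent local that is not read afterward.
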(fn OnBurst(EitherVFX-preferthisversion)
  (params (MyTarget :type Actor%20Object%20Reference) (Parameters :type Gameplay%20Cue%20Parameters%20Structure%20%28by%20ref%29) (ParticleComponents :type Array%20of%20FXSystem%20Component%20Object%20References) (AudioComponents :type Array%20of%20Audio%20Component%20Object%20References) (BurstCameraShakeInstance :type Camera%20Shake%20Base%20Object%20Reference) (BurstDecalInstance :type Decal%20Actor%20Object%20Reference))
  (bind (_normalizedmagnitude _rawmagnitude _effectcontext) (Ability|GameplayCue|BreakGameplayCueParameters Parameters))
  (|CallParentFunction MyTarget Parameters ParticleComponents AudioComponents BurstCameraShakeInstance BurstDecalInstance)
  (Utilities|FlowControl|Sequence
    (:then_0
      (LocalVariables|SetHitBuildings_BoxExtent (Ability|EffectContext|GetOrigin _effectcontext))
      (bind _aspawn (Utilities|Casting|CastToPawn MyTarget))
      (Transformation|SetWorldRotation (Utilities|Array|Get(acopy) ParticleComponents) (Math|Conversions|RotationFromXVector (* (Math|Vector|GetRightVector (Pawn|GetBaseAimRotation _aspawn)) (Math|Float|SelectFloat -1.0 1.0 (== (Math|Float|Truncate _rawmagnitude) 1)))) false false))
    (:then_1
      (return))))
