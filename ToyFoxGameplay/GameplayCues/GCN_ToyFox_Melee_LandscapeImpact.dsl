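; Burst gameplay cue that plays landscape and water impact effects for ToyFox giant melee swings.
; It reads the latest landscape hit results from the pawn interface, spawning Niagara and sound per hit, swapping to splash assets at the surface for water bodies.
; The cue's AbilityLevel parameter encodes swing side, orienting effects along the aim's right vector or its opposite.
; Any hit also triggers one local camera shake.
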(fn OnBurst(EitherVFX-preferthisversion)
  (params (MyTarget :type Actor%20Object%20Reference) (Parameters :type Gameplay%20Cue%20Parameters%20Structure%20%28by%20ref%29) (ParticleComponents :type Array%20of%20FXSystem%20Component%20Object%20References) (AudioComponents :type Array%20of%20Audio%20Component%20Object%20References) (BurstCameraShakeInstance :type Camera%20Shake%20Base%20Object%20Reference) (BurstDecalInstance :type Decal%20Actor%20Object%20Reference))
  (bind (_normalizedmagnitude _rawmagnitude _effectcontext _matchedtagname _originaltag _aggregatedsourcetags _aggregatedtargettags _location _normal _instigator _effectcauser _sourceobject _physicalmaterial _gameplayeffectlevel _abilitylevel) (Ability|GameplayCue|BreakGameplayCueParameters Parameters))
  (|CallParentFunction MyTarget Parameters ParticleComponents AudioComponents BurstCameraShakeInstance BurstDecalInstance)
  (Utilities|FlowControl|Sequence
    (:then_0
      (LocalVariables|SetIsLeftSwing (Utilities|Select :Index _abilitylevel :Option%200 false :Option%201 true :Option%202 false))
      (bind _aspawn (Utilities|Casting|CastToPawn MyTarget))
      (bind _astoy_fox_giant_pawn_interface (Utilities|Casting|CastToToyFoxGiantPawnInterface MyTarget))
      (bind _hit_results (Class|ToyFoxGiantPawnInterface|GetLatestLandscapeHitResults _astoy_fox_giant_pawn_interface))
      (for _array_element _hit_results
        (if (.bSuccess (Utilities|Casting|CastToFortWaterBodyActor (.HitActor (Collision|BreakHitResult _array_element))))
          (bind _outwatersurfaceinfo (WaterBody|GetWaterSurfaceInfoStruct (Utilities|Casting|CastToFortWaterBodyActor (.HitActor (Collision|BreakHitResult _array_element))) (.ImpactPoint (Collision|BreakHitResult _array_element)) true false false))
          (else
            _))
        (bind _returnvalue (Niagara|SpawnSystematLocation MyTarget (select (.bSuccess (Utilities|Casting|CastToFortWaterBodyActor (.HitActor (Collision|BreakHitResult _array_element)))) "/Game/Athena/DrivableVehicles/FX/Plane/WaterInteraction/NS_Biplane_InWater_HugeSplash_Opt.NS_Biplane_InWater_HugeSplash_Opt" "/ToyFoxGameplay/Assets/FX/NS_ToyFox_PowerSowrd_LandImpact_Directional.NS_ToyFox_PowerSowrd_LandImpact_Directional") (select (.bSuccess (Utilities|Casting|CastToFortWaterBodyActor (.HitActor (Collision|BreakHitResult _array_element)))) (.WaterSurfaceLocation (Utilities|Struct|BreakWaterSurfaceInfo _outwatersurfaceinfo)) (.ImpactPoint (Collision|BreakHitResult _array_element))) (Math|Conversions|RotationFromXVector (* (Math|Vector|GetRightVector (Pawn|GetBaseAimRotation _aspawn)) (Math|Integer|SelectInt -1 1 (LocalVariables|GetIsLeftSwing))))))
        (Effects|Components|ParticleSystem|SetFloatParameter _returnvalue "EffectScale" 2.0)
        (Audio|SpawnSoundatLocation MyTarget (select (.bSuccess (Utilities|Casting|CastToFortWaterBodyActor (.HitActor (Collision|BreakHitResult _array_element)))) "/FNGameplayCues/Sounds/Physics/Objects/WaterSplash/LargeObject/PhysicsImpact_Water_Large.PhysicsImpact_Water_Large" "/Game/Athena/Sounds/Vehicles/ShipCannon/Cannon_Explosion_CannonBall_Cue.Cannon_Explosion_CannonBall_Cue") (.ImpactPoint (Collision|BreakHitResult _array_element))))
      (if (> (Utilities|Array|Length _hit_results) 0)
        (CallFunction|PlayLocalCameraShake _aspawn "/ToyFoxGameplay/Gameplay/Ability/Melee/Camera/CameraShake_ToyFox_Melee_Hit_Landscape.CameraShake_ToyFox_Melee_Hit_Landscape_C")))
    (:then_1
      (return))))
