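; Burst gameplay cue for the ToyFox giant pawn's outro rez-out, dissolving the character with masked materials alongside a teleport VFX.
; When cosmetic events can run, OnExecute builds dynamic masked materials for body and sword (sword last), splitting them by index between the two meshes.
; It sets body-mesh dissolve stage, duration and delayed game time for shader-driven animation, then calls the parent.
; OnBurst restarts the spawned Niagara effect with the Megazord teleport flag and a local offset.
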(fn OnBurst(EitherVFX-preferthisversion)
  (params (MyTarget :type Actor%20Object%20Reference) (Parameters :type Gameplay%20Cue%20Parameters%20Structure%20%28by%20ref%29) (ParticleComponents :type Array%20of%20FXSystem%20Component%20Object%20References) (AudioComponents :type Array%20of%20Audio%20Component%20Object%20References) (BurstCameraShakeInstance :type Camera%20Shake%20Base%20Object%20Reference) (BurstDecalInstance :type Decal%20Actor%20Object%20Reference))
  (|CallParentFunction MyTarget Parameters ParticleComponents AudioComponents BurstCameraShakeInstance BurstDecalInstance)
  (bind _asniagara_particle_system_component (Utilities|Casting|CastToNiagaraComponent (Utilities|Array|Get(acopy) ParticleComponents))
    (:then
      (Niagara|SetNiagaraVariable(Bool) _asniagara_particle_system_component "IsMegazordTeleport" true)
      (Components|Activation|Deactivate _asniagara_particle_system_component)
      (Transformation|AddLocalOffset _asniagara_particle_system_component "0.000000,12.278000,0.000000" false false)
      (Components|Activation|SetActive _asniagara_particle_system_component true true))
    (:CastFailed
      _))
  (return))

(fn OnExecute
  (params (MyTarget :type Actor%20Object%20Reference) (Parameters :type Gameplay%20Cue%20Parameters%20Structure%20%28by%20ref%29))
  (bind _astoy_fox_giant_pawn_interface (Utilities|Casting|CastToToyFoxGiantPawnInterface MyTarget))
  (bind _mesh (Class|ToyFoxGiantPawnInterface|GetSkeletalMeshBudgeted _astoy_fox_giant_pawn_interface))
  (bind _weaponmesh (Class|ToyFoxGiantPawnInterface|GetWeaponMeshComponent _astoy_fox_giant_pawn_interface))
  (Networking|CanExecuteCosmeticEvents
    (:True
      (Utilities|FlowControl|Sequence
        (:then_0
          (Utilities|Array|Clear (Variables|Default|GetMID_MaskedOutroMaterials))
          (for _array_element (Variables|Default|GetMaskedOutroMaterials)
            (bind _returnvalue (Rendering|Material|CreateDynamicMaterialInstance MyTarget _array_element "None" "None" :!class /Script/Engine.KismetMaterialLibrary))
            (Utilities|Array|Add (Variables|Default|GetMID_MaskedOutroMaterials) _returnvalue))
          (bind _returnvalue_2 (Rendering|Material|CreateDynamicMaterialInstance MyTarget (Variables|Default|GetSword_MaskedOutroMaterial) "None" "None" :!class /Script/Engine.KismetMaterialLibrary))
          (Utilities|Array|Add (Variables|Default|GetMID_MaskedOutroMaterials) _returnvalue_2) ; Add sword MID To last index
        )
        (:then_1
          (for (_array_element_2 _array_index) (Variables|Default|GetMID_MaskedOutroMaterials)
            (if (<= _array_index 8)
              (Rendering|Material|SetMaterial _mesh _array_index _array_element_2)
              (else
                (Rendering|Material|SetMaterial _weaponmesh 0 _array_element_2))))
          (Rendering|Material|SetScalarParameterValueonMaterials _mesh "Outro_Animation_Duration" 1.0) ; Set Duration. 1s
          (Rendering|Material|SetScalarParameterValueonMaterials _mesh "Outro_Stage" 2.0) ; Stage1
          (Rendering|Material|SetScalarParameterValueonMaterials _mesh "Outro_GameTimeFromBP" (+ (Utilities|Time|GetGameTimeinSeconds MyTarget) (Variables|Default|GetDissolveDelay))) ; Passthough Time
          (return (|CallParentFunction MyTarget Parameters)))))))
