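; Burst gameplay cue for the ToyFox rocket projectile's explosive hit, named for world impacts, that adjusts the spawned effect to the surface it hit.
; After the parent OnBurst runs, it takes the first particle component, if valid, and points its Z axis along the impact normal from the effect context's hit result.
; It also shrinks that effect slightly through its "Global Scale" float parameter.
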(fn OnBurst(EitherVFX-preferthisversion)
  (params (MyTarget :type Actor%20Object%20Reference) (Parameters :type Gameplay%20Cue%20Parameters%20Structure%20%28by%20ref%29) (ParticleComponents :type Array%20of%20FXSystem%20Component%20Object%20References) (AudioComponents :type Array%20of%20Audio%20Component%20Object%20References) (BurstCameraShakeInstance :type Camera%20Shake%20Base%20Object%20Reference) (BurstDecalInstance :type Decal%20Actor%20Object%20Reference))
  (bind (_normalizedmagnitude _rawmagnitude _effectcontext) (Ability|GameplayCue|BreakGameplayCueParameters Parameters))
  (bind (_bblockinghit _binitialoverlap _time _distance _location _impactpoint _normal _impactnormal) (Collision|BreakHitResult (Ability|EffectContext|GetHitResult _effectcontext)))
  (bind _output (Utilities|Array|Get(acopy) ParticleComponents))
  (|CallParentFunction MyTarget Parameters ParticleComponents AudioComponents BurstCameraShakeInstance BurstDecalInstance)
  (Utilities|FlowControl|Sequence
    (:then_0
      (if (Utilities|Array|IsValidIndex ParticleComponents)
        (Utilities|IsValid _output
          (:Is%20Valid
            (Transformation|SetWorldRotation _output (Math|Rotator|MakeRotfromZ _impactnormal) false false)
            (Effects|Components|ParticleSystem|SetFloatParameter _output "Global Scale" 0.75)))))
    (:then_1
      (return))))
