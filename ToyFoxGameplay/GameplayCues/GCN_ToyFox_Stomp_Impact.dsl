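; Gameplay cue notify for the ToyFox stomp impact, playing effects at the stomping foot's socket.
; After the parent burst, it plays the stomp sound and indicator, then line-traces vertically for ground or water.
; The first spawned particle moves to that point (socket as fallback), with a radial physics field if the console variable enables it.
; On water, a scaled splash and camera lens droplets replace that particle, using the water surface height.
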(fn OnBurst(EitherVFX-preferthisversion)
  (params (MyTarget :type Actor%20Object%20Reference) (Parameters :type Gameplay%20Cue%20Parameters%20Structure%20%28by%20ref%29) (ParticleComponents :type Array%20of%20FXSystem%20Component%20Object%20References) (AudioComponents :type Array%20of%20Audio%20Component%20Object%20References) (BurstCameraShakeInstance :type Camera%20Shake%20Base%20Object%20Reference) (BurstDecalInstance :type Decal%20Actor%20Object%20Reference))
  (bind _output (Utilities|Array|Get(acopy) ParticleComponents))
  (bind _stompfootsocketname (Variables|Default|GetStompFootSocketName))
  (bind _stompfootsocketlocation (LocalVariables|GetStompFootSocketLocation))
  (|CallParentFunction MyTarget Parameters ParticleComponents AudioComponents BurstCameraShakeInstance BurstDecalInstance)
  (Utilities|FlowControl|Sequence
    (:then_0
      (bind _asfort_player_pawn (Utilities|Casting|CastToFortPlayerPawn MyTarget))
      (bind _mesh (Variables|/Script/Engine.Character|GetMesh _asfort_player_pawn)
        (:then
          (LocalVariables|SetStompFootSocketLocation (Transformation|GetSocketLocation _mesh _stompfootsocketname))
          (LocalVariables|SetStompFootSocketRotation (Transformation|GetSocketRotation _mesh _stompfootsocketname))
          (Audio|SpawnSoundatLocation _asfort_player_pawn (Variables|Default|GetStompSoundCue) (LocalVariables|GetStompFootSocketLocation))
          (Fortnite|ShowSoundIndicatorLocalforTeam _asfort_player_pawn _asfort_player_pawn (LocalVariables|GetStompFootSocketLocation) (Variables|/Script/Engine.SoundBase|GetMaxDistance (Variables|Default|GetStompSoundCue)) "Gunshot" (CallFunction|GetLocalPilotControllers _asfort_player_pawn) (Team|GetActorTeam _asfort_player_pawn) (Utilities|Array|MakeArray "Friendly" "Neutral" "Hostile") _asfort_player_pawn "(R=0.000000,G=0.000000,B=0.000000,A=0.000000)" "" (Variables|Default|GetStompSoundCue))
          (bind _outhits (Collision|MultiLineTraceForObjects _asfort_player_pawn (+ (LocalVariables|GetStompFootSocketLocation) "0.000000,0.000000,750.000000") (- (LocalVariables|GetStompFootSocketLocation) "0.000000,0.000000,1000.000000") (Utilities|Array|MakeArray "ObjectTypeQuery1" "ObjectTypeQuery13") false (Utilities|Array|MakeArray _asfort_player_pawn) "ForDuration"))
          (bind (_bvalid _impactlocation _bwater) (CallFunction|ParseGroundTraceHitResults _outhits))
          (Utilities|FlowControl|Sequence
            (:then_0
              (Transformation|SetWorldLocation _output (+ (select _bvalid _impactlocation _stompfootsocketlocation) "0.000000,0.000000,5.000000") false false)
              (bind _returnvalue (Development|GetConsoleVariableIntValue "r.PhysicsField.EnableField"))
              (if (== _returnvalue 1)
                (Game|SpawnActorfromClass 0 "/Game/Effects/Niagara/Blueprints/PhsyicsFields/FN_RadialForce.FN_RadialForce_C" (Math|Transform|MakeTransform (select _bvalid _impactlocation _stompfootsocketlocation)) "Undefined" "MultiplyWithRoot" 0 0.4 0.75 2.8e+03 1.4 0.0 0)))
            (:then_1
              (if _bwater
                (bind _returnvalue_2 (Effects|Components|ParticleSystem|SpawnFXAtLocation 0 "/Game/Athena/DrivableVehicles/FX/Plane/WaterInteraction/NS_Biplane_InWater_HugeSplash_Opt.NS_Biplane_InWater_HugeSplash_Opt" (select _bvalid _impactlocation _stompfootsocketlocation) (Math|Rotator|MakeRotfromX (* (Math|Vector|GetRightVector (LocalVariables|GetStompFootSocketRotation)) -1))))
                (Camera|AddGenericCameraLensEffect (Game|GetPlayerCameraManager (Utilities|Casting|CastToFortPlayerPawn MyTarget)) "/Game/Effects/Fort_Effects/Camera_Effects/B_CameraLens_Splash_Water_Droplets.B_CameraLens_Splash_Water_Droplets_C")
                (Effects|Components|ParticleSystem|SetFloatParameter _returnvalue_2 "EffectScale" 2.5)
                (Components|Activation|Deactivate _output)
                (Components|DestroyComponent _output 0)))))
        (:else)))
    (:then_1
      (return))))

(fn ParseGroundTraceHitResults
  (params (HitResults :type Array%20of%20Hit%20Result%20Structures))
  (bind _output (Utilities|Array|Get(aref) HitResults))
  (bind (_bblockinghit _binitialoverlap _time _distance _location _impactpoint _normal _impactnormal _physmat _hitactor) (Collision|BreakHitResult _output))
  (bind (_asfort_water_body_actor _bsuccess) (Utilities|Casting|CastToFortWaterBodyActor _hitactor))
  (Utilities|FlowControl|Sequence
    (:then_0
      (if (Utilities|Array|IsValidIndex HitResults)
        (if _bsuccess
          (bind _outwatersurfaceinfo (WaterBody|GetWaterSurfaceInfoStruct _asfort_water_body_actor _impactpoint true false false))
          (else
            _))
        (return true (select _bsuccess (.WaterSurfaceLocation (Utilities|Struct|BreakWaterSurfaceInfo _outwatersurfaceinfo)) _impactpoint) _bsuccess)))
    (:then_1
      (return false "0, 0, 0" false))))
